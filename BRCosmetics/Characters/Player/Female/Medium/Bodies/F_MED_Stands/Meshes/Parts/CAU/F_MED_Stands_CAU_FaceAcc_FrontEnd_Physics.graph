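# Physics asset for the F_MED_Stands CAU FaceAcc FrontEnd mesh, giving hair bones capsule bodies and joints for simulated hair motion.
# Front, front-top and left, middle and right back hair chains hang from base bodies jointed to a head, upper-spine chain.
# Joints from each base to head and along the spine are free in swing and twist.
# Joints within each chain have tight limits, tightest on the front-top strand.
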PhysicsAsset /BRCosmetics/Characters/Player/Female/Medium/Bodies/F_MED_Stands/Meshes/Parts/CAU/F_MED_Stands_CAU_FaceAcc_FrontEnd_Physics
bodies {
 hairback_00_l [{shapeType=Capsule, length=2, radius=2, shapeName=hairback_02_r_capsule, center=<-2.9642 0 0.0001>, rotation=<rot 90 0 0>}]
 hairback_00_mid [{shapeType=Capsule, length=2, radius=2, shapeName=hairback_02_r_capsule, center=<3.313 0 0>, rotation=<rot 90 0 0>}]
 hairback_00_r [{shapeType=Capsule, length=2, radius=2, shapeName=hairback_02_r_capsule, center=<3.313 0 0>, rotation=<rot 90 0 0>}]
 hairback_01_l [{shapeType=Capsule, length=2, radius=2, shapeName=hairback_02_r_capsule, center=<-2.9642 0 0.0001>, rotation=<rot 90 0 0>}]
 hairback_01_mid [{shapeType=Capsule, length=2, radius=2, shapeName=hairback_02_r_capsule, center=<3.313 0 0>, rotation=<rot 90 0 0>}]
 hairback_01_r [{shapeType=Capsule, length=2, radius=2, shapeName=hairback_02_r_capsule, center=<3.313 0 0>, rotation=<rot 90 0 0>}]
 hairback_02_l [{shapeType=Capsule, length=2, radius=2, shapeName=hairback_02_r_capsule, center=<-2.9642 0 0.0001>, rotation=<rot 90 0 0>}]
 hairback_02_mid [{shapeType=Capsule, length=5.25, radius=2, shapeName=hairback_02_r_capsule, center=<4.7283 0 0>, rotation=<rot 90 0 0>}]
 hairback_02_r [{shapeType=Capsule, length=2, radius=2, shapeName=hairback_02_r_capsule, center=<3.313 0 0>, rotation=<rot 90 0 0>}]
 hairback_03_l [{shapeType=Capsule, length=2, radius=2, shapeName=hairback_02_r_capsule, center=<-2.9642 0 0.0001>, rotation=<rot 90 0 0>}]
 hairback_03_mid [{shapeType=Capsule, length=2, radius=2, shapeName=hairback_02_r_capsule, center=<3.313 0 0>, rotation=<rot 90 0 0>}]
 hairback_03_r [{shapeType=Capsule, length=2, radius=2, shapeName=hairback_02_r_capsule, center=<3.313 0 0>, rotation=<rot 90 0 0>}]
 hairback_base_l [{shapeType=Capsule, length=0.505, radius=0.505, shapeName=hairback_base_l_capsule, rotation=<rot 0 -90 0>}]
 hairback_base_mid [{shapeType=Capsule, length=0.505, radius=0.505, shapeName=hairback_base_mid_capsule, rotation=<rot 0 -90 0>}]
 hairback_base_r [{shapeType=Capsule, length=0.505, radius=0.505, shapeName=hairback_base_r_capsule, rotation=<rot 0 -90 0>}]
 hairfront_01 [{shapeType=Capsule, length=2, radius=2, shapeName=hairfront_01_capsule, center=<2.5981 0 0.0001>, rotation=<rot 90 0 0>}]
 hairfront_02 [{shapeType=Capsule, length=2, radius=2, shapeName=hairfront_02_capsule, center=<2.8274 0 0>, rotation=<rot 90 0 0>}]
 hairfront_03 [{shapeType=Capsule, length=1.75, radius=2.25, shapeName=hairfront_03_capsule, center=<3.1614 0 0>, rotation=<rot 90 0 0>}]
 hairfront_04 [{shapeType=Capsule, length=2.25, radius=1.75, shapeName=hairfront_04_capsule, center=<2.9527 0 0>, rotation=<rot 90 0 0>}]
 hairfront_base [{shapeType=Capsule, length=1, radius=1, shapeName=hairfront_base_capsule}]
 hairfront_top_01 [{shapeType=Capsule, length=2, radius=2, shapeName=hairfront_top_01_capsule, rotation=<rot 90 0 0>}]
 hairfront_top_02 [{shapeType=Capsule, length=2, radius=2, shapeName=hairfront_top_02_capsule, rotation=<rot 90 0 0>}]
 hairfront_top_03 [{shapeType=Capsule, length=2, radius=2, shapeName=hairfront_top_03_capsule, rotation=<rot 90 0 0>}]
 hairfront_top_base [{shapeType=Capsule, length=1, radius=1, shapeName=hairfront_top_base_capsule}]
 head [{shapeType=Capsule, length=8.75, radius=7.0874, shapeName=head_capsule, rotation=<rot 90 14.0363 -0.9641>}]
 spine_04 [{shapeType=Capsule, length=28.0686, radius=7.5423, shapeName=spine_04_capsule, rotation=<rot 0 -90 0>}]
 spine_05 [{shapeType=Capsule, length=29.9237, radius=7.5423, shapeName=spine_05_capsule, rotation=<rot 0 -90 0>}]
}
constraints [
 {bone1Name=hairfront_01, bone2Name=hairfront_base, swing1LimitDegrees=15, swing2LimitDegrees=15, twistLimitDegrees=5}
 {bone1Name=hairfront_02, bone2Name=hairfront_01, swing1LimitDegrees=15, swing2LimitDegrees=15, twistLimitDegrees=5}
 {bone1Name=hairfront_03, bone2Name=hairfront_02, swing1LimitDegrees=15, swing2LimitDegrees=15, twistLimitDegrees=5}
 {bone1Name=hairfront_04, bone2Name=hairfront_03, swing1LimitDegrees=15, swing2LimitDegrees=15, twistLimitDegrees=5}
 {bone1Name=hairfront_top_01, bone2Name=hairfront_top_base, swing1LimitDegrees=5, swing2LimitDegrees=5, twistLimitDegrees=5}
 {bone1Name=hairfront_top_02, bone2Name=hairfront_top_01, swing1LimitDegrees=5, swing2LimitDegrees=5, twistLimitDegrees=5}
 {bone1Name=hairfront_top_03, bone2Name=hairfront_top_02, swing1LimitDegrees=5, swing2LimitDegrees=5, twistLimitDegrees=5}
 {
  bone1Name=hairfront_base
  bone2Name=head
  swing1LimitDegrees=45
  swing1Motion=Free
  swing2LimitDegrees=45
  swing2Motion=Free
  twistLimitDegrees=45
  twistMotion=Free
 }
 {
  bone1Name=hairfront_top_base
  bone2Name=head
  swing1LimitDegrees=45
  swing1Motion=Free
  swing2LimitDegrees=45
  swing2Motion=Free
  twistLimitDegrees=45
  twistMotion=Free
 }
 {
  bone1Name=spine_05
  bone2Name=spine_04
  swing1LimitDegrees=45
  swing1Motion=Free
  swing2LimitDegrees=45
  swing2Motion=Free
  twistLimitDegrees=45
  twistMotion=Free
 }
 {
  bone1Name=head
  bone2Name=spine_05
  swing1LimitDegrees=45
  swing1Motion=Free
  swing2LimitDegrees=45
  swing2Motion=Free
  twistLimitDegrees=45
  twistMotion=Free
 }
 {
  bone1Name=hairback_base_mid
  bone2Name=head
  swing1LimitDegrees=45
  swing1Motion=Free
  swing2LimitDegrees=45
  swing2Motion=Free
  twistLimitDegrees=45
  twistMotion=Free
 }
 {bone1Name=hairback_00_mid, bone2Name=hairback_base_mid, swing1LimitDegrees=15, swing2LimitDegrees=15, twistLimitDegrees=15}
 {bone1Name=hairback_01_mid, bone2Name=hairback_00_mid, swing1LimitDegrees=15, swing2LimitDegrees=15, twistLimitDegrees=15}
 {bone1Name=hairback_02_mid, bone2Name=hairback_01_mid, swing1LimitDegrees=15, swing2LimitDegrees=15, twistLimitDegrees=15}
 {bone1Name=hairback_03_mid, bone2Name=hairback_02_mid, swing1LimitDegrees=15, swing2LimitDegrees=15, twistLimitDegrees=15}
 {
  bone1Name=hairback_base_l
  bone2Name=head
  swing1LimitDegrees=45
  swing1Motion=Free
  swing2LimitDegrees=45
  swing2Motion=Free
  twistLimitDegrees=45
  twistMotion=Free
 }
 {bone1Name=hairback_00_l, bone2Name=hairback_base_l, swing1LimitDegrees=15, swing2LimitDegrees=15, twistLimitDegrees=15}
 {bone1Name=hairback_01_l, bone2Name=hairback_00_l, swing1LimitDegrees=15, swing2LimitDegrees=15, twistLimitDegrees=15}
 {bone1Name=hairback_02_l, bone2Name=hairback_01_l, swing1LimitDegrees=15, swing2LimitDegrees=15, twistLimitDegrees=15}
 {bone1Name=hairback_03_l, bone2Name=hairback_02_l, swing1LimitDegrees=15, swing2LimitDegrees=15, twistLimitDegrees=15}
 {
  bone1Name=hairback_base_r
  bone2Name=head
  swing1LimitDegrees=45
  swing1Motion=Free
  swing2LimitDegrees=45
  swing2Motion=Free
  twistLimitDegrees=45
  twistMotion=Free
 }
 {bone1Name=hairback_00_r, bone2Name=hairback_base_r, swing1LimitDegrees=15, swing2LimitDegrees=15, twistLimitDegrees=15}
 {bone1Name=hairback_01_r, bone2Name=hairback_00_r, swing1LimitDegrees=15, swing2LimitDegrees=15, twistLimitDegrees=15}
 {bone1Name=hairback_02_r, bone2Name=hairback_01_r, swing1LimitDegrees=15, swing2LimitDegrees=15, twistLimitDegrees=15}
 {bone1Name=hairback_03_r, bone2Name=hairback_02_r, swing1LimitDegrees=15, swing2LimitDegrees=15, twistLimitDegrees=15}
]
