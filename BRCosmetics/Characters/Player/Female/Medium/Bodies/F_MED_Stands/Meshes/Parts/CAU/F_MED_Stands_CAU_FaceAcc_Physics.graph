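# Physics asset for the CAU FaceAcc part of the female medium Stands character, defining capsule bodies for hair physics.
# Back hair (left, middle, right), front hair, top-front hair and a right shoulder-helper form bone chains rooted at the head.
# Chain links have tight swing and twist limits, stiffest on top-front hair.
# Spine, neck, head and clavicle capsules approximate the upper body, joined, like most hair roots, by angularly free constraints.
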PhysicsAsset /BRCosmetics/Characters/Player/Female/Medium/Bodies/F_MED_Stands/Meshes/Parts/CAU/F_MED_Stands_CAU_FaceAcc_Physics
bodies {
 clavicle_l [{shapeType=Capsule, length=9.755, radius=4.505, shapeName=clavicle_l_capsule, center=<4.6578 -2.2791 -0.1713>, rotation=<rot 90 0 0>}]
 clavicle_r [{shapeType=Capsule, length=14.505, radius=2.755, shapeName=clavicle_r_capsule, center=<-2.6383 -0.1941 -3.3694>, rotation=<rot 90 0 0>}]
 hairback_00_l [{shapeType=Capsule, length=3.5, radius=2.5, shapeName=hairback_00_l_capsule, center=<-3.1098 1 0>, rotation=<rot 90 0 0>}]
 hairback_00_mid [{shapeType=Capsule, length=3.5, radius=2.5, shapeName=hairback_00_mid_capsule, center=<3.0203 -1.5 0>, rotation=<rot 90 0 0>}]
 hairback_00_r [{shapeType=Capsule, length=3.5, radius=2.5, shapeName=hairback_00_r_capsule, center=<3.0203 -1.5 0>, rotation=<rot 90 0 0>}]
 hairback_01_l [
  {shapeType=Capsule, length=3.25, radius=2, shapeName=hairfront_01_capsule, center=<-3.144 -0.0003 0>, rotation=<rot 90 0 0>}
  {
   shapeType=Capsule
   length=5
   radius=1.75
   shapeName=hairback_01_l_capsule
   center=<-4.1537 -1.9079 3.9532>
   rotation=<rot 84.9998 169.9985 179.9998>
  }
 ]
 hairback_01_mid [{shapeType=Capsule, length=2.25, radius=2, shapeName=hairback_01_mid_capsule, center=<3.4999 0 0>, rotation=<rot 90 0 0>}]
 hairback_01_r [{shapeType=Capsule, length=2.25, radius=2.25, shapeName=hairfront_01_capsule, center=<3.1327 0.044 0>, rotation=<rot 90 0 0>}]
 hairback_02_l [{shapeType=Capsule, length=1.25, radius=2.5, shapeName=hairfront_02_capsule, center=<-2.4487 0.0001 0>, rotation=<rot 90 0 0>}]
 hairback_02_mid [{shapeType=Capsule, length=4, radius=2.5, shapeName=hairback_02_mid_capsule, center=<4.7119 0 0>, rotation=<rot 90 0 0>}]
 hairback_02_r [{shapeType=Capsule, length=1.5, radius=2.75, shapeName=hairfront_02_capsule, center=<2.8694 0.0412 0>, rotation=<rot 90 0 0>}]
 hairback_03_l [{shapeType=Capsule, length=4.25, radius=1.5, shapeName=hairfront_03_capsule, center=<-3.5336 0.0003 0>, rotation=<rot 90 0 0>}]
 hairback_03_mid [{shapeType=Capsule, length=4.25, radius=2, shapeName=hairback_03_mid_capsule, center=<2.0016 0 0>, rotation=<rot 90 0 0>}]
 hairback_03_r [{shapeType=Capsule, length=3.25, radius=1.75, shapeName=hairfront_03_capsule, center=<2.7946 0 0>, rotation=<rot 90 0 0>}]
 hairback_04_r [{shapeType=Capsule, length=2.75, radius=1.5, shapeName=hairfront_04_capsule, center=<3.1543 0 0>, rotation=<rot 90 0 0>}]
 hairback_base_l [{shapeType=Capsule, length=1, radius=1, shapeName=hairback_base_l_capsule}]
 hairback_base_mid [{shapeType=Capsule, length=1, radius=1, shapeName=hairback_base_mid_capsule}]
 hairback_base_r [{shapeType=Capsule, length=0.505, radius=0.505, shapeName=hairback_base_r_capsule, rotation=<rot 0 -90 0>}]
 hairback_shoulder_helper00_r [{shapeType=Capsule, length=2, radius=2, shapeName=hairback_shoulder_helper00_r_capsule, center=<2.0539 0 0>, rotation=<rot 90 0 0>}]
 hairback_shoulder_helper01_r [{shapeType=Capsule, length=2, radius=2, shapeName=hairback_shoulder_helper01_r_capsule, center=<2.0539 0 0>, rotation=<rot 90 0 0>}]
 hairback_shoulder_helper02_r [{shapeType=Capsule, length=2, radius=2, shapeName=hairback_shoulder_helper02_r_capsule, center=<2.0539 0 0>, rotation=<rot 90 0 0>}]
 hairfront_01 [{shapeType=Capsule, length=2, radius=2, shapeName=hairfront_01_capsule, center=<2.5981 0 0.0001>, rotation=<rot 90 0 0>}]
 hairfront_02 [{shapeType=Capsule, length=2, radius=2, shapeName=hairfront_02_capsule, center=<2.8274 0 0>, rotation=<rot 90 0 0>}]
 hairfront_03 [{shapeType=Capsule, length=1.75, radius=2.25, shapeName=hairfront_03_capsule, center=<3.1614 0 0>, rotation=<rot 90 0 0>}]
 hairfront_04 [{shapeType=Capsule, length=2.25, radius=1.75, shapeName=hairfront_04_capsule, center=<2.9527 0 0>, rotation=<rot 90 0 0>}]
 hairfront_base [{shapeType=Capsule, length=1, radius=1, shapeName=hairfront_base_capsule}]
 hairfront_top_01 [{shapeType=Capsule, length=2, radius=2, shapeName=hairfront_top_01_capsule, rotation=<rot 90 0 0>}]
 hairfront_top_02 [{shapeType=Capsule, length=2, radius=2, shapeName=hairfront_top_02_capsule, rotation=<rot 90 0 0>}]
 hairfront_top_03 [{shapeType=Capsule, length=2, radius=2, shapeName=hairfront_top_03_capsule, rotation=<rot 90 0 0>}]
 hairfront_top_base [{shapeType=Capsule, length=1, radius=1, shapeName=hairfront_top_base_capsule}]
 head [{shapeType=Capsule, length=12, radius=7.75, shapeName=head_capsule, rotation=<rot 90 -0.0099 -0.0099>}]
 neck_01 [{
   shapeType=Capsule
   length=32.4237
   radius=2.3411
   shapeName=neck_01_capsule
   center=<1.6887 1.1635 0.2501>
   rotation=<rot -5 -89.9999 -5.0001>
  }]
 spine_04 [{shapeType=Capsule, length=28.0686, radius=7.5423, shapeName=spine_04_capsule, rotation=<rot 0 -90 0>}]
 spine_05 [
  {shapeType=Capsule, length=29.9237, radius=7.5423, shapeName=spine_05_capsule, center=<1.0321 -1.5822 0>, rotation=<rot 0 -89.9999 0>}
  {shapeType=Capsule, length=25.5, radius=4.5, shapeName=spine_05_capsule, center=<4.1499 -1.2197 0>}
 ]
}
constraints [
 {bone1Name=hairback_01_l, bone2Name=hairback_00_l, swing1LimitDegrees=20, swing2LimitDegrees=15, twistLimitDegrees=5}
 {bone1Name=hairback_02_l, bone2Name=hairback_01_l, swing1LimitDegrees=15, swing2LimitDegrees=9, twistLimitDegrees=5}
 {bone1Name=hairback_03_l, bone2Name=hairback_02_l, swing1LimitDegrees=15, swing2LimitDegrees=6, twistLimitDegrees=5}
 {bone1Name=hairfront_01, bone2Name=hairfront_base, swing1LimitDegrees=15, swing2LimitDegrees=15, twistLimitDegrees=5}
 {bone1Name=hairfront_02, bone2Name=hairfront_01, swing1LimitDegrees=15, swing2LimitDegrees=15, twistLimitDegrees=5}
 {bone1Name=hairfront_03, bone2Name=hairfront_02, swing1LimitDegrees=15, swing2LimitDegrees=15, twistLimitDegrees=5}
 {bone1Name=hairfront_04, bone2Name=hairfront_03, swing1LimitDegrees=15, swing2LimitDegrees=15, twistLimitDegrees=5}
 {bone1Name=hairfront_top_01, bone2Name=hairfront_top_base, swing1LimitDegrees=5, swing2LimitDegrees=5, twistLimitDegrees=5}
 {bone1Name=hairfront_top_02, bone2Name=hairfront_top_01, swing1LimitDegrees=5, swing2LimitDegrees=5, twistLimitDegrees=5}
 {bone1Name=hairfront_top_03, bone2Name=hairfront_top_02, swing1LimitDegrees=5, swing2LimitDegrees=5, twistLimitDegrees=5}
 {
  bone1Name=hairback_base_l
  bone2Name=head
  swing1LimitDegrees=45
  swing1Motion=Free
  swing2LimitDegrees=45
  swing2Motion=Free
  twistLimitDegrees=45
  twistMotion=Free
 }
 {
  bone1Name=hairback_base_mid
  bone2Name=head
  swing1LimitDegrees=45
  swing1Motion=Free
  swing2LimitDegrees=45
  swing2Motion=Free
  twistLimitDegrees=45
  twistMotion=Free
 }
 {
  bone1Name=hairfront_base
  bone2Name=head
  swing1LimitDegrees=45
  swing1Motion=Free
  swing2LimitDegrees=45
  swing2Motion=Free
  twistLimitDegrees=45
  twistMotion=Free
 }
 {
  bone1Name=hairfront_top_base
  bone2Name=head
  swing1LimitDegrees=45
  swing1Motion=Free
  swing2LimitDegrees=45
  swing2Motion=Free
  twistLimitDegrees=45
  twistMotion=Free
 }
 {
  bone1Name=spine_05
  bone2Name=spine_04
  swing1LimitDegrees=45
  swing1Motion=Free
  swing2LimitDegrees=45
  swing2Motion=Free
  twistLimitDegrees=45
  twistMotion=Free
 }
 {
  bone1Name=neck_01
  bone2Name=spine_05
  swing1LimitDegrees=45
  swing1Motion=Free
  swing2LimitDegrees=45
  swing2Motion=Free
  twistLimitDegrees=45
  twistMotion=Free
 }
 {bone1Name=hairback_04_r, bone2Name=hairback_03_r, swing1LimitDegrees=15, swing2LimitDegrees=15, twistLimitDegrees=5}
 {bone1Name=hairback_03_r, bone2Name=hairback_02_r, swing1LimitDegrees=20, swing2LimitDegrees=8, twistLimitDegrees=5}
 {bone1Name=hairback_02_r, bone2Name=hairback_01_r, swing1LimitDegrees=20, swing2LimitDegrees=15, twistLimitDegrees=5}
 {bone1Name=hairback_01_r, bone2Name=hairback_00_r, swing1LimitDegrees=20, swing2LimitDegrees=15, twistLimitDegrees=5}
 {
  bone1Name=hairback_base_r
  bone2Name=head
  swing1LimitDegrees=45
  swing1Motion=Free
  swing2LimitDegrees=45
  swing2Motion=Free
  twistLimitDegrees=45
  twistMotion=Free
 }
 {bone1Name=hairback_01_mid, bone2Name=hairback_00_mid, swing1LimitDegrees=15, swing2LimitDegrees=6, twistLimitDegrees=5}
 {bone1Name=hairback_02_mid, bone2Name=hairback_01_mid, swing1LimitDegrees=15, swing2LimitDegrees=6, twistLimitDegrees=5}
 {bone1Name=hairback_03_mid, bone2Name=hairback_02_mid, swing1LimitDegrees=15, swing2LimitDegrees=6, twistLimitDegrees=5}
 {bone1Name=hairback_00_l, bone2Name=hairback_base_l, swing1LimitDegrees=22, swing2LimitDegrees=8, twistLimitDegrees=5}
 {bone1Name=hairback_00_mid, bone2Name=hairback_base_mid, swing1LimitDegrees=15, swing2LimitDegrees=5, twistLimitDegrees=5}
 {bone1Name=hairback_00_r, bone2Name=hairback_base_r, swing1LimitDegrees=22, swing2LimitDegrees=8, twistLimitDegrees=5}
 {
  bone1Name=hairback_shoulder_helper02_r
  bone2Name=hairback_shoulder_helper01_r
  swing1LimitDegrees=20
  swing2LimitDegrees=20
  twistLimitDegrees=5
 }
 {
  bone1Name=hairback_shoulder_helper01_r
  bone2Name=hairback_shoulder_helper00_r
  swing1LimitDegrees=20
  swing2LimitDegrees=20
  twistLimitDegrees=5
 }
 {bone1Name=hairback_shoulder_helper00_r, bone2Name=head, swing1LimitDegrees=20, swing2LimitDegrees=20, twistLimitDegrees=5}
 {
  bone1Name=head
  bone2Name=neck_01
  swing1LimitDegrees=45
  swing1Motion=Free
  swing2LimitDegrees=45
  swing2Motion=Free
  twistLimitDegrees=45
  twistMotion=Free
 }
 {
  bone1Name=clavicle_l
  bone2Name=spine_05
  swing1LimitDegrees=45
  swing1Motion=Free
  swing2LimitDegrees=45
  swing2Motion=Free
  twistLimitDegrees=45
  twistMotion=Free
 }
 {
  bone1Name=clavicle_r
  bone2Name=spine_05
  swing1LimitDegrees=45
  swing1Motion=Free
  swing2LimitDegrees=45
  swing2Motion=Free
  twistLimitDegrees=45
  twistMotion=Free
 }
]
