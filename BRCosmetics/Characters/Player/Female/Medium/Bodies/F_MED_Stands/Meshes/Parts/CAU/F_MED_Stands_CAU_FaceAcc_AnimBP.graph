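# Post-process-style AnimBP for the F_MED_Stands CAU face accessory that animates back hair on the player's attached cosmetic part.
# It copies the parent pose and selects front-end, skydive or default RigidBody simulation, each with its own base bone.
# The default branch layers additive bone correctives after physics, weighted by head pitch, aim and crouch values derived from bone deltas and state.
# Gameplay simulations take a computed gravity override, and dynamics reset after customization loads.
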AnimBlueprint /BRCosmetics/Characters/Player/Female/Medium/Bodies/F_MED_Stands/Meshes/Parts/CAU/F_MED_Stands_CAU_FaceAcc_AnimBP
target_skeleton=&/F_MED_Stands_CAU_FaceAcc_Skeleton
graphs [
 AnimGraph {nodes [
   n1 = "Transform (Modify) Bone - Bone: hairback_base_r"!(ComponentPose, Alpha=1, Rotation="0,-20,0") -> (Pose) {
    rotationMode=BMM_Additive
    rotationSpace=BCS_BoneSpace
    translationMode=BMM_Additive
    translationSpace=BCS_BoneSpace
    alphaBoolBlend {blendOption=Linear}
    boneToModify {boneName=hairback_base_r}
    rotation {yaw=-20}
   }
   n2 = "Transform (Modify) Bone - Bone: hairback_base_l"!(ComponentPose=@n1, Alpha=1, Rotation="0,-20,0") -> (Pose) {
    rotationMode=BMM_Additive
    rotationSpace=BCS_BoneSpace
    translationMode=BMM_Additive
    translationSpace=BCS_BoneSpace
    alphaBoolBlend {blendOption=Linear}
    boneToModify {boneName=hairback_base_l}
    rotation {yaw=-20}
   }
   n3 = "Transform (Modify) Bone - Bone: hairback_base_mid"!(ComponentPose=@n2, Alpha=1, Rotation="0,-20,0") -> (Pose) {
    rotationMode=BMM_Additive
    rotationSpace=BCS_BoneSpace
    translationMode=BMM_Additive
    translationSpace=BCS_BoneSpace
    alphaBoolBlend {blendOption=Linear}
    boneToModify {boneName=hairback_base_mid}
    rotation {yaw=-20}
   }
   "Transform (Modify) Bone - Bone: hairback_shoulder_helper00_r"!(ComponentPose=@n3, Alpha=1, Rotation="-20,0,0") -> (Pose) {
    rotationMode=BMM_Additive
    rotationSpace=BCS_BoneSpace
    translationMode=BMM_Additive
    translationSpace=BCS_BoneSpace
    alphaBoolBlend {blendOption=Linear}
    boneToModify {boneName=hairback_shoulder_helper00_r}
    rotation {pitch=-20}
   }
   n4 = RigidBody!(
    ComponentPose=Local To Component(LocalPose=Copy Pose From Mesh() {bUseAttachedParent=true})
    Alpha
    OverrideWorldGravity:Vector=Get Gravity Override()
    ExternalForce
   ) -> (Pose) {
    bClampLinearTranslationLimitToRefPose=true
    bOverrideWorldGravity=true
    overridePhysicsAsset=&/F_MED_Stands_CAU_FaceAcc_Physics
    simulationSpace=BaseBoneSpace
    alphaBoolBlend {blendOption=Linear}
    baseBoneRef {boneName=spine_03}
    componentAppliedLinearAccClamp=<2500 2500 2500>
    componentLinearAccScale=<0.6 0.6 0.4>
    componentLinearVelScale=<0.6 0.6 0.4>
    overrideWorldGravity {z=-930}
   }
   n5 = Get Head Up() -> (Head Up:double)
   n6 = "Transform (Modify) Bone - Bone: hairback_02_mid"(ComponentPose=@n4, Alpha=@n5, Translation="0,0,0", Rotation="0,15,0", Scale="1,1,1") {
    rotationMode=BMM_Additive
    rotationSpace=BCS_BoneSpace
    translationMode=BMM_Additive
    translationSpace=BCS_BoneSpace
    alphaBoolBlend {blendOption=Linear}
    boneToModify {boneName=hairback_02_mid}
    rotation {yaw=15}
   }
   n7 = "Transform (Modify) Bone - Bone: hairback_02_r"(ComponentPose=@n6, Alpha=@n5, Translation="0,0,0", Rotation="0,15,0", Scale="1,1,1") {
    rotationMode=BMM_Additive
    rotationSpace=BCS_BoneSpace
    translationMode=BMM_Additive
    translationSpace=BCS_BoneSpace
    alphaBoolBlend {blendOption=Linear}
    boneToModify {boneName=hairback_02_r}
    rotation {yaw=15}
   }
   n8 = "Transform (Modify) Bone - Bone: hairback_02_l"(ComponentPose=@n7, Alpha=@n5, Translation="0,0,0", Rotation="0,15,0", Scale="1,1,1") {
    rotationMode=BMM_Additive
    rotationSpace=BCS_BoneSpace
    translationMode=BMM_Additive
    translationSpace=BCS_BoneSpace
    alphaBoolBlend {blendOption=Linear}
    boneToModify {boneName=hairback_02_l}
    rotation {yaw=15}
   }
   n9 = "Transform (Modify) Bone - Bone: hairback_base_r"!(ComponentPose=@n8, Alpha=@n5, Translation:Vector="0,0,0", Rotation="0,-20,0", Scale:Vector="1,1,1") -> (Pose) {
    rotationMode=BMM_Additive
    rotationSpace=BCS_BoneSpace
    translationMode=BMM_Additive
    translationSpace=BCS_BoneSpace
    alphaBoolBlend {blendOption=Linear}
    boneToModify {boneName=hairback_base_r}
    rotation {yaw=-20}
   }
   n10 = "Transform (Modify) Bone - Bone: hairback_shoulder_helper01_r"(ComponentPose=@n9, Alpha=@n5, Translation="0,0,0", Rotation="-60,0,0", Scale="1,1,1") {
    rotationMode=BMM_Additive
    rotationSpace=BCS_BoneSpace
    translationMode=BMM_Additive
    translationSpace=BCS_BoneSpace
    alphaBoolBlend {blendOption=Linear}
    boneToModify {boneName=hairback_shoulder_helper01_r}
    rotation {pitch=-60}
   }
   n11 = "Transform (Modify) Bone - Bone: hairback_01_r"(ComponentPose=@n10, Alpha=@n5, Translation="0,0,0", Rotation="-9.324,30.318,-2.531", Scale="1,1,1") {
    rotationMode=BMM_Additive
    rotationSpace=BCS_BoneSpace
    translationMode=BMM_Additive
    translationSpace=BCS_BoneSpace
    alphaBoolBlend {blendOption=Linear}
    boneToModify {boneName=hairback_01_r}
    rotation=<rot -9.324 30.3181 -2.531>
   }
   n12 = Get Aim Up() -> (Aim Up:double)
   n13 = "Transform (Modify) Bone - Bone: hairback_shoulder_helper00_r"!(ComponentPose=@n11, Alpha=@n12, Translation:Vector="0,0,0", Rotation="-15,0.000,0.000", Scale:Vector="1,1,1") -> (Pose) {
    rotationMode=BMM_Additive
    rotationSpace=BCS_BoneSpace
    translationMode=BMM_Additive
    translationSpace=BCS_BoneSpace
    alphaBoolBlend {blendOption=Linear}
    boneToModify {boneName=hairback_shoulder_helper00_r}
    rotation {pitch=-15}
   }
   n14 = "Transform (Modify) Bone - Bone: hairback_shoulder_helper01_r"(ComponentPose=@n13, Alpha=@n12, Translation="-1,0,-2.5", Rotation="0,0,0", Scale="1,1,1") {
    rotationMode=BMM_Additive
    rotationSpace=BCS_BoneSpace
    translationMode=BMM_Additive
    translationSpace=BCS_BoneSpace
    alphaBoolBlend {blendOption=Linear}
    boneToModify {boneName=hairback_shoulder_helper01_r}
    translation {x=-1, z=-2.5}
   }
   n15 = "Transform (Modify) Bone - Bone: hairback_shoulder_helper02_r"(ComponentPose=@n14, Alpha=@n12, Translation="-2,0,0", Rotation="10,0.000,0.000", Scale="0,1,1") {
    rotationMode=BMM_Additive
    rotationSpace=BCS_BoneSpace
    translationMode=BMM_Additive
    translationSpace=BCS_BoneSpace
    alphaBoolBlend {blendOption=Linear}
    boneToModify {boneName=hairback_shoulder_helper02_r}
    rotation {pitch=10}
    scale {x=0}
    translation {x=-2}
   }
   n16 = "Transform (Modify) Bone - Bone: hairback_02_r"(ComponentPose=@n15, Alpha=@n12, Translation="-2.715,0,0", Rotation="19.961,5.259,0.926", Scale="1,1,1") {
    rotationMode=BMM_Additive
    rotationSpace=BCS_BoneSpace
    translationMode=BMM_Additive
    translationSpace=BCS_BoneSpace
    alphaBoolBlend {blendOption=Linear}
    boneToModify {boneName=hairback_02_r}
    rotation=<rot 19.961 5.2591 0.926>
    translation {x=-2.715}
   }
   n17 = "Transform (Modify) Bone - Bone: hairback_01_r"(ComponentPose=@n16, Alpha=@n12, Translation="0,-2,0", Rotation="0,0,0", Scale="1,1,1") {
    rotationMode=BMM_Additive
    rotationSpace=BCS_BoneSpace
    translationMode=BMM_Additive
    translationSpace=BCS_BoneSpace
    alphaBoolBlend {blendOption=Linear}
    boneToModify {boneName=hairback_01_r}
    translation {y=-2}
   }
   n18 = "Transform (Modify) Bone - Bone: hairback_02_mid"(ComponentPose=@n17, Alpha=@n12, Translation="0,0,0", Rotation="0,15,0", Scale="1,1,1") {
    rotationMode=BMM_Additive
    rotationSpace=BCS_BoneSpace
    translationMode=BMM_Additive
    translationSpace=BCS_BoneSpace
    alphaBoolBlend {blendOption=Linear}
    boneToModify {boneName=hairback_02_mid}
    rotation {yaw=15}
   }
   n19 = "Transform (Modify) Bone - Bone: hairback_02_l"(ComponentPose=@n18, Alpha=@n12, Translation="0,0,0", Rotation="0,15,0", Scale="1,1,1") {
    rotationMode=BMM_Additive
    rotationSpace=BCS_BoneSpace
    translationMode=BMM_Additive
    translationSpace=BCS_BoneSpace
    alphaBoolBlend {blendOption=Linear}
    boneToModify {boneName=hairback_02_l}
    rotation {yaw=15}
   }
   n20 = "Transform (Modify) Bone - Bone: hairback_base_r"(ComponentPose=@n19, Alpha=@n12, Translation="-2,0,0", Rotation="-15,0.000,0.000", Scale="1,1,1") {
    rotationMode=BMM_Additive
    rotationSpace=BCS_BoneSpace
    translationMode=BMM_Additive
    translationSpace=BCS_BoneSpace
    alphaBoolBlend {blendOption=Linear}
    boneToModify {boneName=hairback_base_r}
    rotation {pitch=-15}
    translation {x=-2}
   }
   n21 = Get crouch_head_up() -> (crouch_head_up:double)
   n22 = "Transform (Modify) Bone - Bone: hairback_base_r"(ComponentPose=@n20, Alpha=@n21, Translation="0,0,0", Rotation="0,0,0", Scale="1,1,1") {
    rotationMode=BMM_Additive
    rotationSpace=BCS_BoneSpace
    translationMode=BMM_Additive
    translationSpace=BCS_BoneSpace
    alphaBoolBlend {blendOption=Linear}
    boneToModify {boneName=hairback_base_r}
   }
   n23 = "Transform (Modify) Bone - Bone: hairback_shoulder_helper00_r"(ComponentPose=@n22, Alpha=1, Translation="0.000,-2,-2", Rotation="0,0,0", Scale="1,1,1") {
    rotationMode=BMM_Additive
    rotationSpace=BCS_BoneSpace
    translationMode=BMM_Additive
    translationSpace=BCS_BoneSpace
    alphaBoolBlend {blendOption=Linear}
    boneToModify {boneName=hairback_shoulder_helper00_r}
    translation {y=-2, z=-2}
   }
   n24 = "Transform (Modify) Bone - Bone: hairback_shoulder_helper01_r"(ComponentPose=@n23, Alpha=@n21, Translation="0,0,0", Rotation="0,0,0", Scale="1,1,1") {
    rotationMode=BMM_Additive
    rotationSpace=BCS_BoneSpace
    translationMode=BMM_Additive
    translationSpace=BCS_BoneSpace
    alphaBoolBlend {blendOption=Linear}
    boneToModify {boneName=hairback_shoulder_helper01_r}
   }
   n25 = "Transform (Modify) Bone - Bone: hairback_01_r"(ComponentPose=@n24, Alpha=@n21, Translation="0,0,0", Rotation="0,8,0", Scale="1,1,1") {
    rotationMode=BMM_Additive
    rotationSpace=BCS_BoneSpace
    translationMode=BMM_Additive
    translationSpace=BCS_BoneSpace
    alphaBoolBlend {blendOption=Linear}
    boneToModify {boneName=hairback_01_r}
    rotation {yaw=8}
   }
   n26 = Get crouch_aim_up() -> (crouch_aim_up:double)
   n27 = "Transform (Modify) Bone - Bone: hairback_shoulder_helper00_r"(ComponentPose=@n25, Alpha=@n26, Translation="-1,-2,-1", Rotation="0,0,0", Scale="1,1,1") {
    rotationMode=BMM_Additive
    rotationSpace=BCS_BoneSpace
    translationMode=BMM_Additive
    translationSpace=BCS_BoneSpace
    alphaBoolBlend {blendOption=Linear}
    boneToModify {boneName=hairback_shoulder_helper00_r}
    translation=<-1 -2 -1>
   }
   n28 = "Transform (Modify) Bone - Bone: hairback_shoulder_helper01_r"(ComponentPose=@n27, Alpha=@n26, Translation="0,0,0", Rotation="0,0,0", Scale="0,1,1") {
    rotationMode=BMM_Additive
    rotationSpace=BCS_BoneSpace
    translationMode=BMM_Additive
    translationSpace=BCS_BoneSpace
    alphaBoolBlend {blendOption=Linear}
    boneToModify {boneName=hairback_shoulder_helper01_r}
    scale {x=0}
   }
   n29 = "Transform (Modify) Bone - Bone: hairback_shoulder_helper02_r"(ComponentPose=@n28, Alpha=@n26, Translation="-3,0,0", Rotation="0,0,0", Scale="0,1,1") {
    rotationMode=BMM_Additive
    rotationSpace=BCS_BoneSpace
    translationMode=BMM_Additive
    translationSpace=BCS_BoneSpace
    alphaBoolBlend {blendOption=Linear}
    boneToModify {boneName=hairback_shoulder_helper02_r}
    scale {x=0}
    translation {x=-3}
   }
   n30 = "Transform (Modify) Bone - Bone: hairback_02_r"(ComponentPose=@n29, Alpha=@n26, Translation="0,0,1", Rotation="0,10,9", Scale="1,1,1") {
    rotationMode=BMM_Additive
    rotationSpace=BCS_BoneSpace
    translationMode=BMM_Additive
    translationSpace=BCS_BoneSpace
    alphaBoolBlend {blendOption=Linear}
    boneToModify {boneName=hairback_02_r}
    rotation {roll=9, yaw=10}
    translation {z=1}
   }
   n31 = "Transform (Modify) Bone - Bone: hairback_02_mid"(ComponentPose=@n30, Alpha=@n26, Translation="0,0,0", Rotation="0.000,-10,0.000", Scale="1,1,1") {
    rotationMode=BMM_Additive
    rotationSpace=BCS_BoneSpace
    translationMode=BMM_Additive
    translationSpace=BCS_BoneSpace
    alphaBoolBlend {blendOption=Linear}
    boneToModify {boneName=hairback_02_mid}
    rotation {yaw=-10}
   }
   n32 = "Transform (Modify) Bone - Bone: hairback_02_l"(ComponentPose=@n31, Alpha=@n26, Translation="0,0,0", Rotation="0,0,0", Scale="1,1,1") {
    rotationMode=BMM_Additive
    rotationSpace=BCS_BoneSpace
    translationMode=BMM_Additive
    translationSpace=BCS_BoneSpace
    alphaBoolBlend {blendOption=Linear}
    boneToModify {boneName=hairback_02_l}
   }
   n33 = "Transform (Modify) Bone - Bone: hairback_base_r"(ComponentPose=@n32, Alpha=@n26, Translation="0,0,0", Rotation="0,0,0", Scale="1,1,1") {
    rotationMode=BMM_Additive
    rotationSpace=BCS_BoneSpace
    translationMode=BMM_Additive
    translationSpace=BCS_BoneSpace
    alphaBoolBlend {blendOption=Linear}
    boneToModify {boneName=hairback_base_r}
   }
   n34 = "Transform (Modify) Bone - Bone: hairback_01_r"(ComponentPose=@n33, Alpha=@n26, Translation="0,-2,0", Rotation="0,0,0", Scale="1,1,1") {
    rotationMode=BMM_Additive
    rotationSpace=BCS_BoneSpace
    translationMode=BMM_Additive
    translationSpace=BCS_BoneSpace
    alphaBoolBlend {blendOption=Linear}
    boneToModify {boneName=hairback_01_r}
    translation {y=-2}
   }
   n35 = "Transform (Modify) Bone - Bone: hair_shoulder_r_helper"(ComponentPose=@n34, Alpha=1, Translation="0,0,0", Rotation=Get head_rot() -> (head_rot:Rotator), Scale="1,1,1") {rotationMode=BMM_Additive, rotationSpace=BCS_BoneSpace, alphaBoolBlend {blendOption=Linear}, boneToModify {boneName=hair_shoulder_r_helper}}
   n36 = Get Head Down() -> (Head Down:double)
   n37 = "Transform (Modify) Bone - Bone: hairback_shoulder_helper00_r"(ComponentPose=@n35, Alpha=@n36, Translation="0,0,0", Rotation="0,10,0.000", Scale="0,1,1") {
    rotationMode=BMM_Additive
    rotationSpace=BCS_BoneSpace
    translationMode=BMM_Additive
    translationSpace=BCS_BoneSpace
    alphaBoolBlend {blendOption=Linear}
    boneToModify {boneName=hairback_shoulder_helper00_r}
    rotation {yaw=10}
    scale {x=0}
   }
   n38 = "Transform (Modify) Bone - Bone: hairback_base_r"(ComponentPose=@n37, Alpha=@n36, Translation="0,0,0", Rotation="0,15,0", Scale="1,1,1") {
    rotationMode=BMM_Additive
    rotationSpace=BCS_BoneSpace
    translationMode=BMM_Additive
    translationSpace=BCS_BoneSpace
    alphaBoolBlend {blendOption=Linear}
    boneToModify {boneName=hairback_base_r}
    rotation {yaw=15}
   }
   n39 = "Transform (Modify) Bone - Bone: hairback_base_mid"!(ComponentPose=@n38, Alpha=@n36, Translation:Vector="0,0,0", Rotation="0,15,0", Scale:Vector="1,1,1") -> (Pose) {
    rotationMode=BMM_Additive
    rotationSpace=BCS_BoneSpace
    translationMode=BMM_Additive
    translationSpace=BCS_BoneSpace
    alphaBoolBlend {blendOption=Linear}
    boneToModify {boneName=hairback_base_mid}
    rotation {yaw=15}
   }
   n40 = "Transform (Modify) Bone - Bone: hairback_base_l"!(ComponentPose=@n39, Alpha=@n36, Translation:Vector="0,0,0", Rotation="0,20,0", Scale:Vector="1,1,1") -> (Pose) {
    rotationMode=BMM_Additive
    rotationSpace=BCS_BoneSpace
    translationMode=BMM_Additive
    translationSpace=BCS_BoneSpace
    alphaBoolBlend {blendOption=Linear}
    boneToModify {boneName=hairback_base_l}
    rotation {yaw=20}
   }
   n41 = Component To Local(ComponentPose=@n40)
   n42 = RigidBody(ComponentPose=Local To Component(LocalPose=Copy Pose From Mesh() {bUseAttachedParent=true}), OverrideWorldGravity=Get Gravity Override()) {
    bOverrideWorldGravity=true
    overridePhysicsAsset=&F_MED_Stands_CAU_FaceAcc_Physics
    simulationSpace=BaseBoneSpace
    alphaBoolBlend {blendOption=Linear}
    baseBoneRef {boneName=head}
    componentAppliedLinearAccClamp=<2000 2000 2000>
    componentLinearAccScale {z=0.35}
    componentLinearVelScale {z=0.35}
    overrideWorldGravity {z=-980}
   }
   n43 = Component To Local(ComponentPose=@n42)
   n44 = Get skydive_floating_parachute() -> (skydive_floating_parachute:bool)
   n45 = RigidBody!(ComponentPose=Local To Component(LocalPose=Copy Pose From Mesh() {bUseAttachedParent=true}), Alpha, ExternalForce) -> (Pose) {
    bOverrideWorldGravity=true
    overridePhysicsAsset=&/F_MED_Stands_CAU_FaceAcc_FrontEnd_Physics
    simulationSpace=BaseBoneSpace
    alphaBoolBlend {blendOption=Linear}
    baseBoneRef {boneName=root}
    componentAppliedLinearAccClamp=<2000 2000 2000>
    componentLinearAccScale {z=0.35}
    componentLinearVelScale {z=0.35}
    overrideWorldGravity {z=-980}
   }
   Output Pose(Result=Blend Poses by bool(
     BlendPose_0=Component To Local(ComponentPose=@n45)
     BlendPose_1=Blend Poses by bool(BlendPose_0=@n43, BlendPose_1=@n41, bActiveValue=@n44) {blendType=Linear}
     bActiveValue=Get bIsInFrontEnd()
    ) {blendType=Linear}) {layerGroup=None}
  ]}
 EventGraph {nodes [
   Event BlueprintUpdateAnimation()
   > n1 = Sequence!(execute) -> (then_0, then_1, then_2:Exec, then_3:Exec)
   n2 = GetOwnerSkeletalMeshComponent()
   n3 = BreakRotator(InRot=BreakTransform(InTransform=GetDeltaTransformFromRefPose(self=@n2, BoneName=head, BaseName=neck_02)).Rotation)
   > Set Head Up(Head Up:double=MapRangeClamped(Value=@n3.Yaw, InRangeA=0, InRangeB=-18)) -> (Output_Get:double)
   > Set Head Down(Head Down:double=MapRangeClamped(Value=@n3.Yaw, InRangeA=0, InRangeB=18)) -> (Output_Get:double)
   > Set head_rot(head_rot:Rotator=MakeRotator(Roll=@n3.Yaw, Yaw=@n3)) -> (Output_Get:Rotator)
   n4 = BreakTransform(InTransform=GetDeltaTransformFromRefPose(self=GetOwnerSkeletalMeshComponent(), BoneName=clavicle_l, BaseName=spine_05))
   > Set Aim Up(Aim Up:double=MapRangeClamped(Value=BreakRotator(InRot=@n4.Rotation).Pitch, InRangeA=0, InRangeB=43)) -> (Output_Get:double)
   n5 = Branch(execute=@n1.then_1, Condition=Get bIsCrouching())
   > n6 = Branch(Condition=Get bIsCrouchSprinting())
   Set crouch_head_up(execute=@n6.else, crouch_head_up:double=Get Head Up()) -> (Output_Get:double)
   > Set crouch_head_down(crouch_head_down:double=Get Head Down()) -> (Output_Get:double)
   > Set crouch_aim_up(crouch_aim_up:double=Get Aim Up()) -> (Output_Get:double)
   Set crouch_head_up(execute=@n5.else, crouch_head_up=0)
   > Set crouch_head_down(crouch_head_down=0)
   > Set crouch_aim_up(crouch_aim_up=0)
   Set skydive_floating_parachute(
    execute=@n1.then_2
    skydive_floating_parachute:bool=OR Boolean!(A=Get bIsSkydiveDiveMode(), B=Get bIsSkydiveFloating(), C:bool=Get bIsSkyDiving()) -> (ReturnValue)
   ) -> (Output_Get:bool)
   n7 = Gravity Override State Function DEPRECATED(
    execute=@n1.then_3
    CharacterPartAnimInstance=Self-Reference()
    GameTimeInSeconds=Get Delta Time X() -> (Delta Time X:double)
    GravityOverrideParams=Get Gravity Override Params()
   )
   > Set Gravity Override(Gravity Override=@n7)
   Event OnCharacterCustomizationLoaded()
   > Delay(Duration=0.2)
   > ResetDynamics(InTeleportType=TeleportPhysics)
   > ResetDynamics(InTeleportType=ResetPhysics)
  ]}
]
variables [
 Aim Up {type=number}
 crouch_aim_up {type=number}
 crouch_head_down {type=number}
 crouch_head_up {type=number}
 Delta Time X {type=number}
 Gravity Override {type=Vector, default=<0 0 -980>}
 Gravity Override Params {type=GravityOverrideParamsStruct}
 Head Down {type=number}
 Head Up {type=number}
 head_rot {type=Rotator}
 skydive_floating_parachute {type=boolean}
]
